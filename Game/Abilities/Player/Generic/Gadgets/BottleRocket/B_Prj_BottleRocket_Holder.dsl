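; Bottle rocket gadget holder projectile: a thrown firecracker pack that lands, then launches bottle rocket projectiles.
; Non-pawn hits on shallow enough slopes stop its movement, then it aligns to the surface and starts rocket spawning.
; Rockets fire with random spread and force feedback, while a counter's RepNotify plays muzzle flashes.
; Damage or EndPlay triggers the explosion on authority and effects and sounds on remotes.
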(event Collision|EventHit
  (params (MyComp :type Primitive%20Component%20Object%20Reference) (Other :type Actor%20Object%20Reference) (OtherComp :type Primitive%20Component%20Object%20Reference) (bSelfMoved :type Boolean) (HitLocation :type Vector :default "0, 0, 0") (HitNormal :type Vector :default "0, 0, 0") (NormalImpulse :type Vector :default "0, 0, 0") (Hit :type Hit%20Result%20Structure%20%28by%20ref%29))
  (Utilities|Casting|CastToFortPawn Other
    (:CastFailed
      (if (< (Math|Trig|Acos(Degrees) (Math|Vector|DotProduct HitNormal "0, 0,1.000000")) (Variables|Default|GetMaxSlope))
        (Game|Components|ProjectileMovement|StopSimulating (Variables|Movement|GetProjectileMovementComponent) Hit)))))

(event EventOnStop
  (params (Hit :type Hit%20Result%20Structure%20%28by%20ref%29))
  (bind (_bblockinghit _binitialoverlap _time _distance _location _impactpoint _normal _impactnormal) (Collision|BreakHitResult Hit))
  (bind _mesh (Variables|Projectile_Firecracker_Pack|GetMesh))
  ;[1 1 1] StopEvents
  (Transformation|SetActorRotation (Math|Conversions|RotationFromXVector _impactnormal))
  (Transformation|AddRelativeRotation _mesh "0.000000,90.000000,0.000000" false false)
  (Transformation|AddLocalOffset _mesh "-7.000000, 0, 0" false false)
  (Components|Activation|Activate true)
  (Audio|PlaySoundatLocation (Variables|Default|GetPlacementCue) _impactpoint)
  (CallFunction|StartSpawningRockets))

(event EventBeginPlay
  (Utilities|FlowControl|Delay (Variables|Default|GetInitial%20Delay))
  (AddComponent|NODE_AddAudioComponent-0) ; Looping Fire Cracker sound
  (TargetSelection|PushOffHitLocation self (stale EventOnStop Hit) 50))

(event EventEndPlay
  (params (EndPlayReason :type EEndPlayReason%20Enum :default "Destroyed"))
  (bind _returnvalue (Transformation|GetActorLocation))
  (Utilities|FlowControl|DoOnce false
    (:Completed
      (Networking|SwitchHasAuthority
        (:Authority
          ;[1 0 0.01768] Explosion
          (Delivery|DoExplosion) ; Explosion
        )
        (:Remote
          (Effects|SpawnEmitteratLocation (Variables|Default|GetP_Explosion) _returnvalue (Transformation|GetActorRotation))
          (Audio|PlaySoundatLocation (Variables|Default|GetExplosionCue) _returnvalue)
          (Audio|PlaySoundatLocation (Variables|Default|GetSecondExplosionCue) _returnvalue))) ; If touched or damagenly on server will it trigger explosion
    )))

(event Game|Damage|EventAnyDamage
  (params (Damage :type Float%20%28single-precision%29) (DamageType :type Damage%20Type%20Object%20Reference) (InstigatedBy :type Controller%20Object%20Reference) (DamageCauser :type Actor%20Object%20Reference))
  (bind _returnvalue (Transformation|GetActorLocation))
  (Utilities|FlowControl|DoOnce false
    (:Completed
      (Networking|SwitchHasAuthority
        (:Authority
          ;[1 0 0.01768] Explosion
          (Delivery|DoExplosion) ; Explosion
        )
        (:Remote
          (Effects|SpawnEmitteratLocation (Variables|Default|GetP_Explosion) _returnvalue (Transformation|GetActorRotation))
          (Audio|PlaySoundatLocation (Variables|Default|GetExplosionCue) _returnvalue)
          (Audio|PlaySoundatLocation (Variables|Default|GetSecondExplosionCue) _returnvalue))) ; If touched or damagenly on server will it trigger explosion
    )))

(event Custom|StartSpawningRockets
  (Utilities|FlowControl|Delay (Variables|Default|GetDelayBottleRockets))
  (CallFunction|SpawnBottleRocket (Variables|Default|GetRocketGEContainerSpec))
  (CallFunction|TriggerMuzzleFlashFX)
  (Utilities|FlowControl|Delay (Math|Random|RandomFloatinRange (Variables|Default|GetMin_BR_SpawnDelay) (Variables|Default|GetMax_BR_SpawnDelay))))

(fn OnRep_MuzzleFlashCounter
  (Networking|CanExecuteCosmeticEvents
    (:True
      (Components|Activation|Activate (Variables|Default|GetP_MuzzleFlash) true))))

(fn SpawnBottleRocket
  (params (EffectContainerSpec :type Fort%20Gameplay%20Effect%20Container%20Spec%20Structure))
  (Projectile|SpawnProjectile :ProjectileClass (Variables|Default|GetPrj_BottleRocket) :RequestedBy (Game|GetInstigator) :SpawnLocation (+ (Transformation|GetActorLocation) (* (Math|Vector|Normalize (Math|Rotator|GetRotationXVector (Transformation|GetActorRotation))) (Variables|Default|GetBR_SpawnLoc_Offset))) :SpawnRotation (Math|Conversions|RotationFromXVector (Math|Vector|MakeVector (+ (.x (Math|Rotator|GetRotationXVector (Transformation|GetActorRotation))) (Math|Random|RandomFloatinRange (Variables|Default|GetSpread) (* (Variables|Default|GetSpread) -1))) (+ (.y (Math|Rotator|GetRotationXVector (Transformation|GetActorRotation))) (Math|Random|RandomFloatinRange (Variables|Default|GetSpread) (* (Variables|Default|GetSpread) -1))) (+ (.z (Math|Rotator|GetRotationXVector (Transformation|GetActorRotation))) (Math|Random|RandomFloatinRange (Variables|Default|GetSpread) (* (Variables|Default|GetSpread) -1))))) :SpawnDirection (Math|Conversions|RotationFromXVector (Math|Vector|MakeVector (+ (.x (Math|Rotator|GetRotationXVector (Transformation|GetActorRotation))) (Math|Random|RandomFloatinRange (Variables|Default|GetSpread) (* (Variables|Default|GetSpread) -1))) (+ (.y (Math|Rotator|GetRotationXVector (Transformation|GetActorRotation))) (Math|Random|RandomFloatinRange (Variables|Default|GetSpread) (* (Variables|Default|GetSpread) -1))) (+ (.z (Math|Rotator|GetRotationXVector (Transformation|GetActorRotation))) (Math|Random|RandomFloatinRange (Variables|Default|GetSpread) (* (Variables|Default|GetSpread) -1))))) :EffectContainerSpecToApplyOnExplode EffectContainerSpec :bUseDefaultPhysics false :InitialSpeed (Variables|Default|GetBottleRocketSpeed) :GravityScale (Variables|Default|GetBottleRocketGravity) :ChargePercent 0.0)
  (Fortnite|PlayLocalForceFeedbackatLocationMulti (Variables|Default|GetForceFeedbackNear) (Variables|Default|GetForceFeedbackFar) (Transformation|GetActorLocation) 350 500 "BottleRocket"))

(fn TriggerMuzzleFlashFX
  (Variables|Default|SetMuzzleFlashCounter (+ (Variables|Default|GetMuzzleFlashCounter) 1)))

(fn ConstructionScript)
